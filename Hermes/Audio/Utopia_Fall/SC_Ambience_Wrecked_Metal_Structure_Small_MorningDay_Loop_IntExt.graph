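# Looping outdoor ambience cue for a small wrecked metal structure in morning/day, covering interior and exterior.
# A quality-level switch plays the quad wave for its first two levels and the stereo wave for the last two.
# Each looping wave runs through an indefinitely looping 30-second enveloper that gently bends pitch, then a modulator.
# Ambient attenuation and a two-instance, prevent-new concurrency setting govern playback.
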SoundCue /Hermes/Audio/Utopia_Fall/SC_Ambience_Wrecked_Metal_Structure_Small_MorningDay_Loop_IntExt
root=QualityLevel_2
nodes [
 {
  id=Enveloper_4
  children [WavePlayer_2]
  properties {
   bLoop=true
   bLoopIndefinitely=true
   loopEnd=30
   pitchCurve {editorCurveData {keys [
      {arriveTangent=-0.0013, interpMode=RCIM_Cubic, leaveTangent=-0.0013, time=0}
      <at 15 0.98 RCIM_Cubic>
      {arriveTangent=0.0013, interpMode=RCIM_Cubic, leaveTangent=0.0013, time=30}
     ]}}
   volumeCurve {editorCurveData {keys [<at 0 1>]}}
  }
 }
 {
  id=Enveloper_6
  children [WavePlayer_3]
  properties {
   bLoop=true
   bLoopIndefinitely=true
   loopEnd=30
   pitchCurve {editorCurveData {keys [
      {arriveTangent=-0.0013, interpMode=RCIM_Cubic, leaveTangent=-0.0013, time=0}
      <at 15 0.98 RCIM_Cubic>
      {arriveTangent=0.0013, interpMode=RCIM_Cubic, leaveTangent=0.0013, time=30}
     ]}}
   volumeCurve {editorCurveData {keys [<at 0 1>]}}
  }
 }
 {id=Modulator_0, children [Enveloper_6]}
 {id=Modulator_1, children [Enveloper_4]}
 {id=QualityLevel_2, children [Modulator_0,Modulator_0,Modulator_1,Modulator_1]}
 {
  id=WavePlayer_2
  properties {bLooping=true, soundWaveAssetPtr=&/Ambience_Wrecked_Metal_Structure_Small_MorningDay_Loop_IntExt_Stereo}
 }
 {
  id=WavePlayer_3
  properties {bLooping=true, soundWaveAssetPtr=&/Ambience_Wrecked_Metal_Structure_Small_MorningDay_Loop_IntExt_Quad}
 }
]
properties {
 attenuationSettings=/AmbientAudioBase/Attenuations/Fort_Ambient_Att
 bEnableBusSends=false
 bEnableSubmixSends=false
 maxDistance=2097152
 priority=7
 soundClassObject=&Fort_Ambient_Outdoors_Athena
 volumeMultiplier=2
 concurrencySet [/AmbientAudioBase/Concurrency/AmbientConc_2Max_Loop_PreventNew_Default_Concurrency]
}
